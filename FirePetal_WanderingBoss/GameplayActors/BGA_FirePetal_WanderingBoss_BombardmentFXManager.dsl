(event EventBeginPlay
  (bind _ns_firepetal_wanderingboss_bombard_aoe_burst_mergedv (Variables|Default|GetNS_FirePetal_WanderingBoss_Bombard_AOE_Burst_MergedV))
  (bind _ns_firepetal_wanderingboss_bombard_aoe_burst_mergedv_2 (Variables|Default|GetNS_FirePetal_WanderingBoss_Bombard_AOE_Burst_MergedV))
  (bind _ns_firepetal_wanderingboss_bombard_aoe_idle_manageversion (Variables|Default|GetNS_FirePetal_WanderingBoss_Bombard_AOE_Idle_ManageVersion))
  (Development|PrintString "SpawnedFXManager" true true "(R=0.000000,G=1.000000,B=0.550351,A=1.000000)" 1e+01)
  (Utilities|Array|Clear (Variables|Default|GetExplosionFXList))
  (Utilities|Array|Clear (Variables|Default|GetSpawnTimeList))
  (Utilities|Array|Clear (Variables|Default|GetIdleFXList))
  (Utilities|Array|Clear (Variables|Default|GetBGAList))
  (Utilities|Array|Clear (Variables|Default|GetExplodedList))
  (Niagara|NiagaraSetPositionArray _ns_firepetal_wanderingboss_bombard_aoe_burst_mergedv_2 "SpawnArray" (Variables|Default|GettmpArray))
  (Niagara|NiagaraSetBoolArray _ns_firepetal_wanderingboss_bombard_aoe_burst_mergedv_2 "ExplodeArray" (Variables|Default|GettmpArrayb))
  (Niagara|NiagaraSetPositionArray _ns_firepetal_wanderingboss_bombard_aoe_burst_mergedv_2 "SpawnRateList" (Variables|Default|GettmpArray))
  (Niagara|NiagaraSetFloatArray _ns_firepetal_wanderingboss_bombard_aoe_burst_mergedv_2 "SpawnRateStartTime" (Variables|Default|GettmpArrayf))
  (Niagara|NiagaraSetPositionArray _ns_firepetal_wanderingboss_bombard_aoe_idle_manageversion "SpawnArray" (Variables|Default|GettmpArray))
  (Niagara|NiagaraSetBoolArray _ns_firepetal_wanderingboss_bombard_aoe_idle_manageversion "ExplodeArray" (Variables|Default|GettmpArrayb))
  (Niagara|NiagaraSetPositionArray _ns_firepetal_wanderingboss_bombard_aoe_idle_manageversion "SpawnRateList" (Variables|Default|GettmpArray))
  (Niagara|NiagaraSetFloatArray _ns_firepetal_wanderingboss_bombard_aoe_idle_manageversion "SpawnRateStartTime" (Variables|Default|GettmpArrayf))
  (Components|Activation|Deactivate _ns_firepetal_wanderingboss_bombard_aoe_burst_mergedv)
  (Niagara|ResetSystem _ns_firepetal_wanderingboss_bombard_aoe_burst_mergedv)
  (Components|Activation|SetActive _ns_firepetal_wanderingboss_bombard_aoe_burst_mergedv true true))

(event EventTick
  (params (DeltaSeconds :type Float%20%28single-precision%29)))

(event Collision|EventActorBeginOverlap
  (params (OtherActor :type Actor%20Object%20Reference)))

(event Custom|Reset
  (bind _explosionfxlist (Variables|Default|GetExplosionFXList))
  (bind _ns_firepetal_wanderingboss_bombard_aoe_burst_mergedv (Variables|Default|GetNS_FirePetal_WanderingBoss_Bombard_AOE_Burst_MergedV))
  (bind _ns_firepetal_wanderingboss_bombard_aoe_burst_mergedv_2 (Variables|Default|GetNS_FirePetal_WanderingBoss_Bombard_AOE_Burst_MergedV))
  (bind _ns_firepetal_wanderingboss_bombard_aoe_idle_manageversion (Variables|Default|GetNS_FirePetal_WanderingBoss_Bombard_AOE_Idle_ManageVersion))
  (if (Variables|Debug|GetbDebugActor)
    (Default|DebugPrintArray _explosionfxlist "(R=0.000000,G=1.000000,B=0.891056,A=1.000000)")
    (else
      _))
  (Utilities|Array|Clear _explosionfxlist)
  (Utilities|Array|Clear (Variables|Default|GetSpawnTimeList))
  (Utilities|Array|Clear (Variables|Default|GetIdleFXList))
  (Utilities|Array|Clear (Variables|Default|GetBGAList))
  (Utilities|Array|Clear (Variables|Default|GetExplodedList))
  (Niagara|NiagaraSetPositionArray _ns_firepetal_wanderingboss_bombard_aoe_burst_mergedv_2 "SpawnArray" (Variables|Default|GettmpArray))
  (Niagara|NiagaraSetBoolArray _ns_firepetal_wanderingboss_bombard_aoe_burst_mergedv_2 "ExplodeArray" (Variables|Default|GettmpArrayb))
  (Niagara|NiagaraSetPositionArray _ns_firepetal_wanderingboss_bombard_aoe_burst_mergedv_2 "SpawnRateList" (Variables|Default|GettmpArray))
  (Niagara|NiagaraSetFloatArray _ns_firepetal_wanderingboss_bombard_aoe_burst_mergedv_2 "SpawnRateStartTime" (Variables|Default|GettmpArrayf))
  (Niagara|NiagaraSetPositionArray _ns_firepetal_wanderingboss_bombard_aoe_idle_manageversion "SpawnArray" (Variables|Default|GettmpArray))
  (Niagara|NiagaraSetBoolArray _ns_firepetal_wanderingboss_bombard_aoe_idle_manageversion "ExplodeArray" (Variables|Default|GettmpArrayb))
  (Niagara|NiagaraSetPositionArray _ns_firepetal_wanderingboss_bombard_aoe_idle_manageversion "SpawnRateList" (Variables|Default|GettmpArray))
  (Niagara|NiagaraSetFloatArray _ns_firepetal_wanderingboss_bombard_aoe_idle_manageversion "SpawnRateStartTime" (Variables|Default|GettmpArrayf))
  (Components|Activation|Deactivate _ns_firepetal_wanderingboss_bombard_aoe_burst_mergedv)
  (Niagara|ResetSystem _ns_firepetal_wanderingboss_bombard_aoe_burst_mergedv)
  (Components|Activation|SetActive _ns_firepetal_wanderingboss_bombard_aoe_burst_mergedv true true))

(event EventDestroyed
  (Development|PrintString "FXManagerDestroyed"))

(fn ClientDoExplode
  (params (Index :type Integer))
  (Utilities|Array|SetArrayElem (Variables|Default|GetExplodedList) Index "true" true)
  (Niagara|NiagaraSetBoolArray (Variables|Default|GetNS_FirePetal_WanderingBoss_Bombard_AOE_Idle_ManageVersion) "ExplodeArray" (Variables|Default|GetExplodedList))
  (CallFunction|UpdateSpawnRateList))

(fn CreateANewExplosionFX
  (params (ExplosionFXPosition :type Vector))
  (Utilities|Array|Add (Variables|Default|GetExplosionFXList) ExplosionFXPosition)
  (Development|PrintString "Created a new Explosion FX" true true "(R=1.000000,G=0.048716,B=0.000000,A=1.000000)" 1e+01)
  (Niagara|NiagaraSetPositionArray (Variables|Default|GetNS_FirePetal_WanderingBoss_Bombard_AOE_Burst_MergedV) "SpawnArray" (Variables|Default|GetExplosionFXList))
  (return))

(fn CreateANewIdleFX
  (params (IdleFXPosition :type Vector) (BGA :type BGA%20Fire%20Petal%20Wandering%20Boss%20Ground%20Explosion%20Object%20Reference))
  (Utilities|Array|Add (Variables|Default|GetIdleFXList) IdleFXPosition)
  (Utilities|Array|Add (Variables|Default|GetBGAList) BGA)
  (Development|PrintString "Created a new IdleFX" true true "(R=1.000000,G=0.000000,B=0.565177,A=1.000000)" 1e+01)
  (return))

(fn DebugPrintArray
  (params (LocationArray :type Array%20of%20Vectors) (LocationColor :type Linear%20Color%20Structure))
  (Development|PrintString (Utilities|String|Append "@@@@@@@@@ DebugPrintArray [BGA_FirePetal_WanderingBoss_BombardmentFXManager]: CALLED with a Location Array of size " (Utilities|String|ToString(Integer) (Utilities|Array|Length (Variables|GetLocationArray))) "!") false (Variables|Debug|GetbDebugActor))
  (for _array_element (Variables|GetLocationArray)
    (Rendering|Debug|DrawDebugSphere _array_element 121.0 12 (Variables|GetLocationColor) 1e+01 2.0)))

(fn OnRep_BGAList
  (for (_array_element _array_index) (Variables|Default|GetBGAList)
    (Variables|/FirePetal_WanderingBoss/GameplayActors/BGA_FirePetal_WanderingBoss_GroundExplosion.BGA_FirePetal_WanderingBoss_GroundExplosion_C|SetMyIndexInFXManager _array_index _array_element)
    (Variables|/FirePetal_WanderingBoss/GameplayActors/BGA_FirePetal_WanderingBoss_GroundExplosion.BGA_FirePetal_WanderingBoss_GroundExplosion_C|SetFXManager self _array_element)))

(fn OnRep_IdleFXList
  (bind _idlefxlist (Variables|Default|GetIdleFXList))
  (bind _returnvalue (Utilities|Array|Length _idlefxlist))
  (Niagara|NiagaraSetPositionArray (Variables|Default|GetNS_FirePetal_WanderingBoss_Bombard_AOE_Idle_ManageVersion) "SpawnArray" _idlefxlist)
  (Utilities|Array|Resize (Variables|Default|GetExplodedList) _returnvalue)
  (for _ (range (Utilities|Array|Length (Variables|Default|GetSpawnTimeList)) _returnvalue)
    (Utilities|Array|Add (Variables|Default|GetSpawnTimeList) (Utilities|Time|GetRealTimeSeconds)))
  (CallFunction|UpdateSpawnRateList))

(fn UpdateSpawnRateList
  (bind _ns_firepetal_wanderingboss_bombard_aoe_idle_manageversion (Variables|Default|GetNS_FirePetal_WanderingBoss_Bombard_AOE_Idle_ManageVersion))
  (Utilities|Array|Clear (LocalVariables|GetSpawnRateList))
  (Utilities|Array|Clear (LocalVariables|GetSpawnRateStartTimeList))
  (for (_array_element _array_index) (Variables|Default|GetExplodedList)
    (if (not _array_element)
      (Utilities|Array|Add (LocalVariables|GetSpawnRateList) (Utilities|Array|Get(acopy) (Variables|Default|GetIdleFXList) _array_index))
      (Utilities|Array|Add (LocalVariables|GetSpawnRateStartTimeList) (Utilities|Array|Get(acopy) (Variables|Default|GetSpawnTimeList) _array_index))))
  (Niagara|NiagaraSetPositionArray _ns_firepetal_wanderingboss_bombard_aoe_idle_manageversion "SpawnRateList" (LocalVariables|GetSpawnRateList))
  (Niagara|NiagaraSetFloatArray _ns_firepetal_wanderingboss_bombard_aoe_idle_manageversion "SpawnRateStartTime" (LocalVariables|GetSpawnRateStartTimeList)))

(fn ConstructionScript)
