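; Timed ground explosion actor for the FirePetal wandering boss's bombardment, dealing fire damage in an area.
; On authority, it waits out a scalable fuse, capsule-traces, applies separate damage to vehicles, players and buildings, and knocks vehicles and pawns away.
; It then sets bActivateGeyser and destroys itself shortly after.
; That flag's OnRep drives the shared FX manager explosion, camera shake, sound and indicator, plus a radial force field on higher-detail non-mobile platforms.
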(event EventBeginPlay
  (Networking|SwitchHasAuthority
    (:Authority
      (Utilities|FlowControl|DoOnce false
        (:Completed
          (Variables|Default|SetAppliedFireDmg 0.0)
          (Variables|Default|SetbActivateGeyser false)
          (Utilities|FlowControl|Delay (ScalableFloat|GetValueatLevel (Variables|Default|GetSF_AOEFuseTime)))
          (bind _outhits (Collision|MultiCapsuleTraceForObjects (Transformation|GetWorldLocation (Variables|Default|GetHitCapsule)) (+ (Transformation|GetWorldLocation (Variables|Default|GetHitCapsule)) "0.000000,0.000000,1.000000") (Components|Capsule|GetScaledCapsuleRadius (Variables|Default|GetHitCapsule)) (Components|Capsule|GetScaledCapsuleHalfHeight (Variables|Default|GetHitCapsule)) (Variables|Default|GetObject%20Types) false (Variables|Default|GetActors%20to%20Ignore)))
          (for _array_element _outhits
            (bind _output_get (Variables|Default|SetCurrentTestActor (.HitActor (Collision|BreakHitResult _array_element))))
            (bind _asfort_vehicle_interface (Utilities|Casting|CastToFortVehicleInterface _output_get)
              (:then
                (Variables|Default|SetAppliedFireDmg (ScalableFloat|GetValueatLevel (Variables|Default|GetSF_AOEDmgVehicle)))
                (Default|KnockbackVehicle (Utilities|Casting|CastToFortAthenaVehicle _asfort_vehicle_interface)))
              (:CastFailed
                (bind _asfort_player_pawn (Utilities|Casting|CastToFortPlayerPawn (Variables|Default|GetCurrentTestActor))
                  (:then
                    (Variables|Default|SetAppliedFireDmg (ScalableFloat|GetValueatLevel (Variables|Default|GetSF_AOEDmgPlayer)))
                    (Default|KnockbackPawn _asfort_player_pawn))
                  (:CastFailed
                    (Utilities|Casting|CastToBuildingActor (Variables|Default|GetCurrentTestActor))
                    (Variables|Default|SetAppliedFireDmg (ScalableFloat|GetValueatLevel (Variables|Default|GetSF_AOEDmgBuilding)))))))
            (if (> (Variables|Default|GetAppliedFireDmg) 0)
              (Environment|ApplyGlobalEnvironmentDamagetoActor (Variables|Default|GetCurrentTestActor) (Variables|Default|GetAppliedFireDmg))))
          (Variables|Default|SetbActivateGeyser true)
          (Utilities|FlowControl|Delay 0.5)
          (Actor|DestroyActor))))))

(fn KnockbackPawn
  (params (FortPawn :type Fort%20Pawn%20Object%20Reference))
  (bind (_returnvalue_x _returnvalue_y) (Math|Vector|BreakVector (Math|Vector|Normalize (- (Transformation|GetActorLocation FortPawn) (Transformation|GetWorldLocation (Variables|Default|GetHitCapsule))))))
  (Character|LaunchCharacter FortPawn (Math|Vector|MakeVector (* _returnvalue_x (ScalableFloat|GetValueatLevel (Variables|Default|GetKnockbackPawnForwardVelocity))) (* _returnvalue_y (ScalableFloat|GetValueatLevel (Variables|Default|GetKnockbackPawnForwardVelocity))) (ScalableFloat|GetValueatLevel (Variables|Default|GetKnockbackPawnHeightVelocity))) true true))

(fn KnockbackVehicle
  (params (FortAthenaVehicle :type Fort%20Athena%20Vehicle%20Object%20Reference))
  (bind (_returnvalue_x _returnvalue_y) (Math|Vector|BreakVector (Math|Vector|Normalize (- (Transformation|GetActorLocation FortAthenaVehicle) (Transformation|GetWorldLocation (Variables|Default|GetHitCapsule))))))
  (Audio|BroadcastAbilityImpactatLocation 0 FortAthenaVehicle (Math|Vector|MakeVector (* _returnvalue_x (ScalableFloat|GetValueatLevel (Variables|Default|GetKnockbackVehicleForwardVelocity))) (* _returnvalue_y (ScalableFloat|GetValueatLevel (Variables|Default|GetKnockbackVehicleForwardVelocity))) (ScalableFloat|GetValueatLevel (Variables|Default|GetKnockbackVehicleHeightVelocity))) 777.0 "None" false))

(fn OnRep_bActivateGeyser
  (bind _self self)
  (bind _returnvalue (Transformation|GetActorLocation _self))
  (bind _returnvalue_2 (+ _returnvalue "0.000000,0.000000,-15.000000"))
  (bind _fxmanager (Variables|Default|GetFXManager))
  (if (Variables|Default|GetbActivateGeyser)
    (Utilities|IsValid _fxmanager
      (:Is%20Valid
        (Class|BGAFirePetalWanderingBossBombardmentFXManager|ClientDoExplode _fxmanager (Variables|Default|GetMyIndexInFXManager))
        (Class|BGAFirePetalWanderingBossBombardmentFXManager|CreateANewExplosionFX _fxmanager _returnvalue_2)
        (Camera|PlayWorldCameraShake "/DemonMaskFire/Assets/VFX/PostFX/B_DemonMaskFire_ImpactShake.B_DemonMaskFire_ImpactShake_C" _returnvalue_2 5e+02 1e+03)
        (Audio|SpawnSoundAttached "/DemonMaskFire/Assets/Audio/DemonMaskFire_WanderingBoss/SC_WanderingBoss_Barrage_Impact.SC_WanderingBoss_Barrage_Impact" (Variables|Default|GetSoundPivot))
        (SoundIndicator|ShowSoundIndicator (Variables|Default|GetFortSoundIndicator) _returnvalue_2 "/FirePetal_WanderingBoss/Assets/Audio/FirePitHazard_Erupt_Burst_Cue.FirePitHazard_Erupt_Burst_Cue" 1.0 0)
        ;[0.1302 0 0] VFX: Radial Force
        (if (Math|Boolean|ANDBoolean (not (Fortnite|IsMobilePlatform)) (not (Class|PlatformHelpersLibrary|IsPcPerfMode)) (> (Rendering|GetRenderingDetailMode) 1))
          (Game|SpawnActorfromClass "/Game/Effects/Niagara/Blueprints/PhsyicsFields/FN_RadialForce.FN_RadialForce_C" (Math|Transform|MakeTransform _returnvalue) "AlwaysSpawn" "OverrideRootScale" 0 1.5 0.8 8.5e+02 0.9 0.0 0)
          (Development|PrintString "Bombardment Radial Force Spawned" true true "(R=1.000000,G=0.000000,B=0.885789,A=1.000000)" 1e+01))))))

(fn ConstructionScript)
