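; Patrol path point provider for the FirePetal wandering boss that configures the boss pawn when it spawns.
; On begin play with authority, it binds to its actor-spawned result event.
; On a successful spawn, it casts the actor to the wandering boss pawn and assigns it the leash volume.
; It also reports the spawn to the boss progress bar challenge manager, if that reference is valid.
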(event EventBeginPlay
  (Networking|SwitchHasAuthority
    (:Authority
      (EventDispatchers|BindEventtoOnActorSpawnedResult (delegate Custom|OnActorSpawned)))))

(event Custom|OnActorSpawned
  (params (SpawnedActor :type Actor%20Object%20Reference) (bResult :type Boolean))
  (bind _progressmanager (Variables|Default|GetProgressManager))
  (if bResult
    (bind _asbp_player_pawn_fire_petal_wandering_boss (Utilities|Casting|CastToBP_PlayerPawn_FirePetal_WanderingBoss SpawnedActor))
    (Variables|/FirePetal_WanderingBoss/NPCs/WanderingBoss/Pawn/BP_PlayerPawn_FirePetal_WanderingBoss.BP_PlayerPawn_FirePetal_WanderingBoss_C|SetLeashVolume (Variables|Default|GetWanderingDemonLeashVolum) _asbp_player_pawn_fire_petal_wandering_boss)
    (Utilities|IsValid _progressmanager
      (:Is%20Valid
        (Class|ChallengeManagerSimpleBossProgressBar|AISpawned _progressmanager SpawnedActor bResult)))))

(fn ConstructionScript)
